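# Control Rig for the BittenHead player head cosmetic that adds procedural secondary jiggle to its big and small leaf bone chains.
# Each chain's tip position is Verlet-integrated, then converted into the chain root's space and compared with the tip's actual offset.
# The resulting rotation is slerped toward identity and spread along the chain by Distribute Rotation with quadratic ease-in.
# The two leaves use separately tuned spring and damping settings.
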ControlRigBlueprint /BRCosmetics/Characters/Player/Male/Medium/Heads/M_MED_BittenHead_Head/Meshes/BittenHead_Head_ControlRig
graphs [
 RigVMFunctionLibrary {links [], nodes []}
 RigVMModel {nodes [
   RigUnit_BeginExecution = &RigUnit_BeginExecution()
   n1 = &RigUnit_GetRelativeBoneTransform(Bone=leaf_big_05, Space=leaf_big_01)
   n2 = /Script/RigVM.RigVMFunction_MathTransformTransformVector(
    Transform=/Script/RigVM.RigVMFunction_MathTransformInverse(Value=&RigUnit_GetBoneTransform(Bone=leaf_big_01).Transform) {is=struct RigVMUnitNode, template="Inverse::Execute(in Value,out Result)"}.Result
    Location=/Script/RigVM.RigVMFunction_VerletIntegrateVector(
     Blend=8
     Damp=0.2
     Force=(X=0,Y=0,Z=0)
     Strength=15
     Target=&RigUnit_GetBoneTransform(Bone=leaf_big_05).Transform.Translation
    ) {
     is=struct RigVMUnitNode
     template="Verlet::Execute(in Target,in Strength,in Damp,in Blend,in Force,out Position,out Velocity,out Acceleration)"
    }.Position
   ) {is=struct RigVMUnitNode}
   /Script/ControlRig.RigUnit_DistributeRotation(
    EndBone=leaf_big_05
    RotationEaseType=QuadraticEaseIn
    Rotations=((Rotation=(X=0,Y=0,Z=0,W=1),Ratio=1))
    StartBone=leaf_big_01
    Weight=1
    bPropagateToChildren=True
    Rotations.0.Rotation=/Script/RigVM.RigVMFunction_MathQuaternionSlerp(
     B=(X=0,Y=0,Z=0,W=1)
     T=0.75
     A=/Script/RigVM.RigVMFunction_MathQuaternionFromTwoVectors(A=@n1.Transform.Translation, B=@n2.Result) {is=struct RigVMUnitNode, template="FromTwoVectors::Execute(in A,in B,out Result)"}.Result
    ) {is=struct RigVMUnitNode, template="Interpolate::Execute(in A,in B,in T,out Result)"}.Result
    ExecutePin=@RigUnit_BeginExecution.ExecutePin
   ) {is=struct RigVMUnitNode}
   n3 = &RigUnit_GetRelativeBoneTransform(Bone=leaf_small_04, Space=leaf_small_01)
   n4 = &RigVMFunction_MathTransformTransformVector(
    Transform=&RigVMFunction_MathTransformInverse(Value=&RigUnit_GetBoneTransform(Bone=leaf_small_01).Transform) {template="Inverse::Execute(in Value,out Result)"}.Result
    Location=&RigVMFunction_VerletIntegrateVector(
     Blend=4
     Damp=0.15
     Force=(X=0,Y=0,Z=0)
     Strength=70
     Target=&RigUnit_GetBoneTransform(Bone=leaf_small_04).Transform.Translation
    ) {template="Verlet::Execute(in Target,in Strength,in Damp,in Blend,in Force,out Position,out Velocity,out Acceleration)"}.Position
   )
   > &RigUnit_DistributeRotation(
    EndBone=leaf_small_04
    RotationEaseType=QuadraticEaseIn
    Rotations=((Rotation=(X=0,Y=0,Z=0,W=1),Ratio=1))
    StartBone=leaf_small_01
    Weight=1
    bPropagateToChildren=false
    Rotations.0.Rotation=&RigVMFunction_MathQuaternionSlerp(
     B=(X=0,Y=0,Z=0,W=1)
     T=0.8
     A=&RigVMFunction_MathQuaternionFromTwoVectors(A=@n3.Transform.Translation, B=@n4.Result) {template="FromTwoVectors::Execute(in A,in B,out Result)"}.Result
    ) {template="Interpolate::Execute(in A,in B,in T,out Result)"}.Result
   )
  ]}
]
hierarchy [
 root {loc=<0 0 0>, rot=<0 0 0 1>}
 pelvis {parent=root, loc=<0 -0.7071 93.1191>, rot=<0.0127 0.707 -0.0127 -0.707>}
 spine_01 {parent=pelvis, loc=<6.7494 -0.6388 0>, rot=<0 0 0.034 -0.9994>}
 spine_02 {parent=spine_01, loc=<7.2518 -0.035 0>, rot=<0 0 0.0238 -0.9997>}
 spine_03 {parent=spine_02, loc=<6.0104 0 0>, rot=<0 0 0.0149 -0.9999>}
 spine_04 {parent=spine_03, loc=<8.1024 0 0>, rot=<0 0 -0.0132 -0.9999>}
 spine_05 {parent=spine_04, loc=<10.0146 0 0>, rot=<0 0 -0.0015 -1>}
 clavicle_l {parent=spine_05, loc=<4.3575 -0.1466 -2.3223>, rot=<0.0465 -0.7289 -0.009 -0.6829>}
 clavicle_r {parent=spine_05, loc=<4.3575 -0.1466 2.3223>, rot=<0.7289 0.0465 -0.6829 0.009>}
 neck_01 {parent=spine_05, loc=<11.3138 0.1529 0>, rot=<0 0 -0.2003 -0.9797>}
 neck_02 {parent=neck_01, loc=<4.5627 0 0>, rot=<0 0 0.0504 -0.9987>}
 head {parent=neck_02, loc=<5.3914 0 0>, rot=<0 0 0.115 -0.9934>}
 leaf_small_01 {parent=head, loc=<24.1964 1.6571 2.527>, rot=<0.8304 -0.1592 0.4331 -0.3124>}
 leaf_small_02 {parent=leaf_small_01, loc=<1.2384 0 0>, rot=<-0.0241 -0.2975 -0.077 -0.9513>}
 leaf_small_03 {parent=leaf_small_02, loc=<2.8429 0 0>, rot=<0.0079 -0.2871 -0.0264 -0.9575>}
 leaf_small_04 {parent=leaf_small_03, loc=<2.6491 0 0>, rot=<0 0 0 1>}
 leaf_big_01 {parent=head, loc=<19.3029 -0.0505 -0.1318>, rot=<0.4334 -0.2759 0.1383 -0.8467>}
 leaf_big_02 {parent=leaf_big_01, loc=<3.1099 0 0>, rot=<-0.0453 -0.3083 -0.0632 -0.9481>}
 leaf_big_03 {parent=leaf_big_02, loc=<7.4172 0 0>, rot=<0.0001 -0.2508 0.0015 -0.968>}
 leaf_big_04 {parent=leaf_big_03, loc=<5.1217 0 0>, rot=<0.0001 -0.1619 -0.0002 -0.9868>}
 leaf_big_05 {parent=leaf_big_04, loc=<4.4843 0 0>, rot=<0 0 0 1>}
]
