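# Physics asset for the BittenHead male medium player head cosmetic, giving physics bodies to the head and its leaves.
# One capsule covers the head, and box bodies follow two leaf bone chains, five big segments and four small.
# Angular-limited constraints link each segment to the previous one, with tighter limits on the small leaf.
# No constraint joins either chain's first bone to the head body.
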PhysicsAsset /BRCosmetics/Characters/Player/Male/Medium/Heads/M_MED_BittenHead_Head/Meshes/BittenHead_Head_PhysicsAsset
bodies {
 head [{shapeType=Capsule, length=23.8086, radius=18.7879, center=<-8.215 -0.0001 0>, rotation=<rot 90 0 0>}]
 leaf_big_01 [{shapeType=Box, extentX=1.2204, extentY=1.1913, extentZ=2.5067, center=<1.9628 -0.0285 0.0118>, rotation=<rot 0 0 90>}]
 leaf_big_02 [{shapeType=Box, extentX=7.6267, extentY=1.969, extentZ=9.436, center=<3.3294 0.2334 0.0217>, rotation=<rot 0 0 90>}]
 leaf_big_03 [{shapeType=Box, extentX=5.1435, extentY=2.1237, extentZ=10.1194, center=<2.3242 0.0291 -0.1636>, rotation=<rot 0 0 90>}]
 leaf_big_04 [{shapeType=Box, extentX=3.8364, extentY=2.0498, extentZ=8.7882, center=<1.5039 -0.0343 -0.2944>, rotation=<rot 0 0 90>}]
 leaf_big_05 [{shapeType=Box, extentX=3.714, extentY=1.5157, extentZ=5.3933, center=<1.0048 -0.3796 -0.1488>, rotation=<rot 0 0 90>}]
 leaf_small_01 [{shapeType=Box, extentX=0.5685, extentY=0.4717, extentZ=1.5551, center=<0.6145 0.0019 -0.0049>, rotation=<rot 0 0 90>}]
 leaf_small_02 [{shapeType=Box, extentX=2.3118, extentY=0.7781, extentZ=3.728, center=<1.1699 0.046 -0.0301>, rotation=<rot 0 0 90>}]
 leaf_small_03 [{shapeType=Box, extentX=2.9296, extentY=0.9783, extentZ=3.9928, center=<0.6869 0.0263 -0.0458>, rotation=<rot 0 0 90>}]
 leaf_small_04 [{shapeType=Box, extentX=2.661, extentY=1.2802, extentZ=2.9836, center=<0.9724 -0.0441 -0.0525>, rotation=<rot 0 0 90>}]
}
constraints [
 {bone1Name=leaf_big_02, bone2Name=leaf_big_01, swing1LimitDegrees=12, swing2LimitDegrees=18, twistLimitDegrees=8}
 {bone1Name=leaf_big_03, bone2Name=leaf_big_02, swing1LimitDegrees=12, swing2LimitDegrees=18, twistLimitDegrees=8}
 {bone1Name=leaf_big_04, bone2Name=leaf_big_03, swing1LimitDegrees=12, swing2LimitDegrees=18, twistLimitDegrees=8}
 {bone1Name=leaf_big_05, bone2Name=leaf_big_04, swing1LimitDegrees=12, swing2LimitDegrees=18, twistLimitDegrees=8}
 {bone1Name=leaf_small_02, bone2Name=leaf_small_01, swing1LimitDegrees=6, swing2LimitDegrees=14, twistLimitDegrees=6}
 {bone1Name=leaf_small_03, bone2Name=leaf_small_02, swing1LimitDegrees=6, swing2LimitDegrees=14, twistLimitDegrees=6}
 {bone1Name=leaf_small_04, bone2Name=leaf_small_03, swing1LimitDegrees=6, swing2LimitDegrees=14, twistLimitDegrees=6}
]
