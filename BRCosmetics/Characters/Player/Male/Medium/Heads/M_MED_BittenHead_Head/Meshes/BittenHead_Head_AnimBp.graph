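# Animation Blueprint for the BittenHead player cosmetic head, copying the attached parent mesh's pose and curves and adding secondary physics.
# A per-bone layered blend mixes in the head's idle animation, weighted by the DisableFaceOverride curve value.
# A RigidBody node then simulates the head's physics asset in spine_05 bone space, with overridden gravity and scaled component motion.
# The Control Rig node is unconnected and the EventGraph holds no logic.
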AnimBlueprint /BRCosmetics/Characters/Player/Male/Medium/Heads/M_MED_BittenHead_Head/Meshes/BittenHead_Head_AnimBp
target_skeleton=&/BittenHead_Head_Skeleton
graphs [
 AnimGraph {nodes [
   Control Rig() {
    alphaBoolBlend {blendOption=Linear}
    controlRigAssetReference {blueprintRigClass=&/BittenHead_Head_ControlRig.BittenHead_Head_ControlRig_C}
   }
   n1 = Copy Pose From Mesh() {bCopyCurves=true, bUseAttachedParent=true}
   n2 = Sequence Player 'bittenHead_Idle'() {sequence=&/bittenHead_Idle}
   n3 = float * float(A=1, B=GetCurveValue(CurveName=DisableFaceOverride))
   n4 = RigidBody(ComponentPose=Local To Component(LocalPose=Layered blend per bone(BasePose=@n1, BlendPoses_0=@n2, BlendWeights_0=@n3))) {
    bClampLinearTranslationLimitToRefPose=true
    bOverrideWorldGravity=true
    overridePhysicsAsset=&/BittenHead_Head_PhysicsAsset
    simulationSpace=BaseBoneSpace
    alphaBoolBlend {blendOption=Linear}
    baseBoneRef {boneName=spine_05}
    componentAppliedLinearAccClamp=<1000 1000 2000>
    componentLinearAccScale=<0.4 0.4 0.4>
    componentLinearVelScale=<0.3 3 0.3>
    overrideWorldGravity {z=-980}
   }
   Output Pose(Result=Component To Local(ComponentPose=@n4)) {layerGroup=None}
  ]}
 EventGraph {nodes [
   TryGetPawnOwner()
   Event BlueprintUpdateAnimation()
  ]}
]
variables []
